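# Physics asset for the Critter Frenzy male medium body, rigging the jacket's lower panels for physics around the pelvis and thighs.
# Seven box-bodied bone chains cover back, sides and front, each hung from a small master body joined to the pelvis with free rotation.
# Segments within chains have tight swing and twist limits.
# Freely rotating cross-constraints tie each front chain's tip to its neighboring side chain, while capsules represent pelvis and thighs.
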PhysicsAsset /BRCosmetics/Characters/Player/Male/Medium/Bodies/M_MED_Critter_Frenzy/Meshes/M_MED_Critter_Frenzy_Physics
bodies {
 dyn_jacket_bk_01 [{shapeType=Box, extentX=7.5, extentY=14, extentZ=2, shapeName=dyn_jacket_bk_01_box, center=<3.8154 0 -0.7164>}]
 dyn_jacket_bk_02 [{shapeType=Box, extentX=7.5, extentY=13, extentZ=2.625, shapeName=dyn_jacket_bk_02_box, center=<4.3156 0 -0.7164>}]
 dyn_jacket_bk_MSTR [
  {shapeType=Capsule, length=0.505, radius=0.505, shapeName=dyn_jacket_bk_MSTR_capsule, rotation=<rot 0 -90 0>}
  {shapeType=Box, extentX=1.5, extentY=1.5, extentZ=3, shapeName=dyn_jacket_bk_MSTR_box}
 ]
 dyn_jacket_bk_l_01 [{shapeType=Box, extentX=7.5, extentY=14, extentZ=2, shapeName=dyn_jacket_bk_l_01_box, center=<4.3571 0 -1.0102>}]
 dyn_jacket_bk_l_02 [{shapeType=Box, extentX=7.5, extentY=13, extentZ=2.625, shapeName=dyn_jacket_bk_l_02_box, center=<4.3571 0 -1.0102>}]
 dyn_jacket_bk_l_MSTR [
  {shapeType=Capsule, length=0.505, radius=0.505, shapeName=dyn_jacket_bk_l_MSTR_capsule, rotation=<rot 0 -90 0>}
  {shapeType=Box, extentX=1.5, extentY=1.5, extentZ=3, shapeName=dyn_jacket_bk_l_MSTR_box}
 ]
 dyn_jacket_bk_r_01 [{shapeType=Box, extentX=7.5, extentY=14, extentZ=2, shapeName=dyn_jacket_bk_r_01_box, center=<3.8154 0 -0.8424>}]
 dyn_jacket_bk_r_02 [{shapeType=Box, extentX=7.5, extentY=13, extentZ=2.625, shapeName=dyn_jacket_bk_r_02_box, center=<3.8154 0 -0.8424>}]
 dyn_jacket_bk_r_MSTR [
  {shapeType=Capsule, length=0.505, radius=0.505, shapeName=dyn_jacket_bk_r_MSTR_capsule, rotation=<rot 0 -90 0>}
  {shapeType=Box, extentX=1.5, extentY=1.5, extentZ=3, shapeName=dyn_jacket_bk_r_MSTR_box}
 ]
 dyn_jacket_fr_l_01 [{shapeType=Box, extentX=8.5, extentY=14, extentZ=2, shapeName=dyn_jacket_fr_l_01_box, center=<5.2339 0 -0.0001>}]
 dyn_jacket_fr_l_02 [{shapeType=Box, extentX=8.5, extentY=13, extentZ=2.625, shapeName=dyn_jacket_fr_l_02_box, center=<4.1881 0 -0.0002>}]
 dyn_jacket_fr_l_03 [{shapeType=Box, extentX=8.5, extentY=13, extentZ=3.625, shapeName=dyn_jacket_fr_l_02_box, center=<5.5394 0 -0.0001>}]
 dyn_jacket_fr_l_MSTR [
  {shapeType=Capsule, length=0.505, radius=0.505, shapeName=dyn_jacket_fr_l_MSTR_capsule, rotation=<rot 0 -90 0>}
  {shapeType=Box, extentX=1.5, extentY=1.5, extentZ=3, shapeName=dyn_jacket_fr_l_MSTR_box}
 ]
 dyn_jacket_fr_r_01 [{shapeType=Box, extentX=8.5, extentY=14, extentZ=2, shapeName=dyn_jacket_fr_l_01_box, center=<-4.682 0 -0.0003>}]
 dyn_jacket_fr_r_02 [{shapeType=Box, extentX=7.5, extentY=14, extentZ=2.5, shapeName=dyn_jacket_fr_l_01_box, center=<-3.8807 0 -0.0004>}]
 dyn_jacket_fr_r_03 [{shapeType=Box, extentX=8.5, extentY=14, extentZ=2.5, shapeName=dyn_jacket_fr_l_01_box, center=<-4.8749 0 -0.0003>}]
 dyn_jacket_fr_r_MSTR [{shapeType=Capsule, length=0.505, radius=0.505, shapeName=dyn_jacket_fr_r_MSTR_capsule, rotation=<rot 0 -90 0>}]
 dyn_jacket_l_01 [{shapeType=Box, extentX=7.5, extentY=14, extentZ=2, shapeName=dyn_jacket_l_01_box, center=<4.5049 0 0>}]
 dyn_jacket_l_02 [{shapeType=Box, extentX=5, extentY=13, extentZ=2.625, shapeName=dyn_jacket_l_02_box, center=<3.4223 0 0.0001>}]
 dyn_jacket_l_03 [{shapeType=Box, extentX=8.5, extentY=13, extentZ=2.625, shapeName=dyn_jacket_l_02_box, center=<4.8851 0 0.0002>}]
 dyn_jacket_l_MSTR [
  {shapeType=Capsule, length=0.505, radius=0.505, shapeName=dyn_jacket_l_MSTR_capsule, rotation=<rot 0 -90 0>}
  {shapeType=Box, extentX=1.5, extentY=1.5, extentZ=3, shapeName=dyn_jacket_l_MSTR_box}
 ]
 dyn_jacket_r_01 [{shapeType=Box, extentX=8.5, extentY=14, extentZ=2, shapeName=dyn_jacket_r_01_box, center=<3.8154 0 0>}]
 dyn_jacket_r_02 [{shapeType=Box, extentX=8.5, extentY=13, extentZ=2.625, shapeName=dyn_jacket_r_02_box, center=<3.8154 0 0>}]
 dyn_jacket_r_MSTR [
  {shapeType=Capsule, length=0.505, radius=0.505, shapeName=dyn_jacket_r_MSTR_capsule, rotation=<rot 0 -90 0>}
  {shapeType=Box, extentX=1.5, extentY=1.5, extentZ=3, shapeName=dyn_jacket_r_MSTR_box}
 ]
 pelvis [{shapeType=Capsule, length=10.45, radius=8.75, shapeName=pelvis_capsule, center=<-2.5782 -1.3701 0>}]
 thigh_l [{shapeType=Capsule, length=26.5, radius=9.625, shapeName=thigh_l_capsule, center=<-14 -0.3202 0>, rotation=<rot 90 0 0>}]
 thigh_r [{shapeType=Capsule, length=26.75, radius=8.625, shapeName=thigh_r_capsule, center=<14 0 0>, rotation=<rot 90 0 0>}]
}
constraints [
 {bone1Name=dyn_jacket_l_01, bone2Name=dyn_jacket_l_MSTR, swing1LimitDegrees=15, swing2LimitDegrees=15, twistLimitDegrees=5}
 {bone1Name=dyn_jacket_l_02, bone2Name=dyn_jacket_l_01, swing1LimitDegrees=15, swing2LimitDegrees=15, twistLimitDegrees=5}
 {bone1Name=dyn_jacket_fr_l_01, bone2Name=dyn_jacket_fr_l_MSTR, swing1LimitDegrees=8, swing2LimitDegrees=8, twistLimitDegrees=5}
 {bone1Name=dyn_jacket_fr_l_02, bone2Name=dyn_jacket_fr_l_01, swing1LimitDegrees=8, swing2LimitDegrees=8, twistLimitDegrees=5}
 {bone1Name=dyn_jacket_bk_l_01, bone2Name=dyn_jacket_bk_l_MSTR, swing1LimitDegrees=8, swing2LimitDegrees=8, twistLimitDegrees=5}
 {bone1Name=dyn_jacket_bk_l_02, bone2Name=dyn_jacket_bk_l_01, swing1LimitDegrees=8, swing2LimitDegrees=8, twistLimitDegrees=5}
 {bone1Name=dyn_jacket_bk_01, bone2Name=dyn_jacket_bk_MSTR, swing1LimitDegrees=8, swing2LimitDegrees=8, twistLimitDegrees=5}
 {bone1Name=dyn_jacket_bk_02, bone2Name=dyn_jacket_bk_01, swing1LimitDegrees=8, swing2LimitDegrees=8, twistLimitDegrees=5}
 {bone1Name=dyn_jacket_bk_r_01, bone2Name=dyn_jacket_bk_r_MSTR, swing1LimitDegrees=8, swing2LimitDegrees=8, twistLimitDegrees=5}
 {bone1Name=dyn_jacket_bk_r_02, bone2Name=dyn_jacket_bk_r_01, swing1LimitDegrees=8, swing2LimitDegrees=8, twistLimitDegrees=5}
 {bone1Name=dyn_jacket_r_01, bone2Name=dyn_jacket_r_MSTR, swing1LimitDegrees=15, swing2LimitDegrees=15, twistLimitDegrees=5}
 {bone1Name=dyn_jacket_r_02, bone2Name=dyn_jacket_r_01, swing1LimitDegrees=15, swing2LimitDegrees=15, twistLimitDegrees=5}
 {
  bone1Name=dyn_jacket_l_MSTR
  bone2Name=pelvis
  swing1LimitDegrees=45
  swing1Motion=Free
  swing2LimitDegrees=45
  swing2Motion=Free
  twistLimitDegrees=45
  twistMotion=Free
 }
 {
  bone1Name=dyn_jacket_fr_l_MSTR
  bone2Name=pelvis
  swing1LimitDegrees=45
  swing1Motion=Free
  swing2LimitDegrees=45
  swing2Motion=Free
  twistLimitDegrees=45
  twistMotion=Free
 }
 {
  bone1Name=dyn_jacket_bk_l_MSTR
  bone2Name=pelvis
  swing1LimitDegrees=45
  swing1Motion=Free
  swing2LimitDegrees=45
  swing2Motion=Free
  twistLimitDegrees=45
  twistMotion=Free
 }
 {
  bone1Name=dyn_jacket_bk_MSTR
  bone2Name=pelvis
  swing1LimitDegrees=45
  swing1Motion=Free
  swing2LimitDegrees=45
  swing2Motion=Free
  twistLimitDegrees=45
  twistMotion=Free
 }
 {
  bone1Name=dyn_jacket_bk_r_MSTR
  bone2Name=pelvis
  swing1LimitDegrees=45
  swing1Motion=Free
  swing2LimitDegrees=45
  swing2Motion=Free
  twistLimitDegrees=45
  twistMotion=Free
 }
 {
  bone1Name=dyn_jacket_r_MSTR
  bone2Name=pelvis
  swing1LimitDegrees=45
  swing1Motion=Free
  swing2LimitDegrees=45
  swing2Motion=Free
  twistLimitDegrees=45
  twistMotion=Free
 }
 {
  bone1Name=thigh_l
  bone2Name=pelvis
  swing1LimitDegrees=45
  swing1Motion=Free
  swing2LimitDegrees=45
  swing2Motion=Free
  twistLimitDegrees=45
  twistMotion=Free
 }
 {
  bone1Name=thigh_r
  bone2Name=pelvis
  swing1LimitDegrees=45
  swing1Motion=Free
  swing2LimitDegrees=45
  swing2Motion=Free
  twistLimitDegrees=45
  twistMotion=Free
 }
 {bone1Name=dyn_jacket_fr_l_03, bone2Name=dyn_jacket_fr_l_02, swing1LimitDegrees=8, swing2LimitDegrees=8, twistLimitDegrees=5}
 {bone1Name=dyn_jacket_l_03, bone2Name=dyn_jacket_l_02, swing1LimitDegrees=8, swing2LimitDegrees=8, twistLimitDegrees=5}
 {
  bone1Name=dyn_jacket_fr_r_MSTR
  bone2Name=pelvis
  swing1LimitDegrees=45
  swing1Motion=Free
  swing2LimitDegrees=45
  swing2Motion=Free
  twistLimitDegrees=45
  twistMotion=Free
 }
 {bone1Name=dyn_jacket_fr_r_01, bone2Name=dyn_jacket_fr_r_MSTR, swing1LimitDegrees=8, swing2LimitDegrees=8, twistLimitDegrees=5}
 {bone1Name=dyn_jacket_fr_r_02, bone2Name=dyn_jacket_fr_r_01, swing1LimitDegrees=8, swing2LimitDegrees=8, twistLimitDegrees=5}
 {bone1Name=dyn_jacket_fr_r_03, bone2Name=dyn_jacket_fr_r_02, swing1LimitDegrees=8, swing2LimitDegrees=8, twistLimitDegrees=5}
 {
  bone1Name=dyn_jacket_fr_r_03
  bone2Name=dyn_jacket_r_02
  swing1LimitDegrees=45
  swing1Motion=Free
  swing2LimitDegrees=45
  swing2Motion=Free
  twistLimitDegrees=45
  twistMotion=Free
 }
 {
  bone1Name=dyn_jacket_fr_l_03
  bone2Name=dyn_jacket_l_03
  swing1LimitDegrees=45
  swing1Motion=Free
  swing2LimitDegrees=45
  swing2Motion=Free
  twistLimitDegrees=45
  twistMotion=Free
 }
]
