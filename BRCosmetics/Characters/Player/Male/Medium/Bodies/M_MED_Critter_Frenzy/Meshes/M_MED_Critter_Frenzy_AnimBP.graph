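# Post-copy AnimBP for the Critter Frenzy body cosmetic that keeps the jacket's dynamic bones posed during jumping and crouching.
# It copies the attached parent's pose, layers additive bone offsets and scale tweaks around a pelvis-space RigidBody simulation, then outputs the result.
# The event graph derives offset alphas from jump, fall and crouch states plus thigh rotations relative to pelvis.
# A front-end blend branch exists but is not wired to the output.
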AnimBlueprint /BRCosmetics/Characters/Player/Male/Medium/Bodies/M_MED_Critter_Frenzy/Meshes/M_MED_Critter_Frenzy_AnimBP
target_skeleton=&/M_MED_Critter_Frenzy_Skeleton
graphs [
 AnimGraph {nodes [
   n1 = "Transform (Modify) Bone - Bone: dyn_jacket_fr_r_MSTR"(
    ComponentPose=Local To Component(LocalPose=Copy Pose From Mesh() {bUseAttachedParent=true})
    Alpha=1
    Translation="0,0,0"
    Rotation="10,0,0"
    Scale="1,1,1"
   ) {
    rotationMode=BMM_Additive
    rotationSpace=BCS_BoneSpace
    translationMode=BMM_Additive
    alphaBoolBlend {blendInTime=0.25, blendOption=Linear, blendOutTime=0.25}
    boneToModify {boneName=dyn_jacket_fr_r_MSTR}
    rotation {pitch=10}
   }
   n2 = "Transform (Modify) Bone - Bone: dyn_jacket_fr_l_MSTR"(ComponentPose=@n1, Alpha=1, Translation="0,0,0", Rotation="10,0,0.000", Scale="1,1,1") {
    rotationMode=BMM_Additive
    rotationSpace=BCS_BoneSpace
    translationMode=BMM_Additive
    alphaBoolBlend {blendInTime=0.25, blendOption=Linear, blendOutTime=0.25}
    boneToModify {boneName=dyn_jacket_fr_l_MSTR}
    rotation {pitch=10}
   }
   n3 = RigidBody(ComponentPose=@n2) {
    bClampLinearTranslationLimitToRefPose=true
    bOverrideWorldGravity=true
    overridePhysicsAsset=&/M_MED_Critter_Frenzy_Physics
    simulationSpace=BaseBoneSpace
    alphaBoolBlend {blendOption=Linear}
    baseBoneRef {boneName=pelvis}
    componentAppliedLinearAccClamp=<2000 2000 2000>
    componentLinearAccScale=<0.3 0.3 0.3>
    componentLinearVelScale=<0.3 0.3 0.3>
    overrideWorldGravity {z=-980}
   }
   n4 = "Transform (Modify) Bone - Bone: dyn_jacket_r_01"(ComponentPose=@n3, Alpha=1, Translation="0,0,0", Rotation="-5,0,0", Scale="1,1,1") {
    rotationMode=BMM_Additive
    rotationSpace=BCS_BoneSpace
    translationMode=BMM_Additive
    alphaBoolBlend {blendInTime=0.25, blendOption=Linear, blendOutTime=0.25}
    boneToModify {boneName=dyn_jacket_r_01}
    rotation {pitch=-5}
   }
   n5 = "Transform (Modify) Bone - Bone: dyn_jacket_l_02"(ComponentPose=@n4, Alpha=1, Translation="0,0,0", Rotation="-10,0,0", Scale="1,1,1") {
    rotationMode=BMM_Additive
    rotationSpace=BCS_BoneSpace
    translationMode=BMM_Additive
    alphaBoolBlend {blendInTime=0.25, blendOption=Linear, blendOutTime=0.25}
    boneToModify {boneName=dyn_jacket_l_02}
    rotation {pitch=-10}
   }
   Blend Poses by bool(BlendPose_0=Component To Local(ComponentPose=@n5), bActiveValue=Get bIsInFrontEnd()) {blendType=Linear}
   n6 = Get Thigh L Crouch Pitch Alpha() -> (Thigh L Crouch Pitch Alpha:double)
   "Transform (Modify) Bone - Bone: dyn_jacket_bk_l_MSTR"(Alpha=@n6, Translation="0,0,0", Rotation="20,0,0", Scale="1,1,1") {
    rotationMode=BMM_Additive
    rotationSpace=BCS_BoneSpace
    translationMode=BMM_Additive
    alphaBoolBlend {blendInTime=0.25, blendOption=Linear, blendOutTime=0.25}
    alphaScaleBiasClamp {bInterpResult=true, interpSpeedIncreasing=12}
    boneToModify {boneName=dyn_jacket_bk_l_MSTR}
    rotation {pitch=20}
   }
   RigidBody() {alphaBoolBlend {blendOption=Linear}}
   n7 = Get Jump Alpha() -> (Jump Alpha:double)
   n8 = "Transform (Modify) Bone - Bone: dyn_jacket_fr_l_MSTR"(
    ComponentPose=Local To Component(LocalPose=Copy Pose From Mesh() {bUseAttachedParent=true})
    Alpha=@n7
    Translation="0,0,0"
    Rotation="30,0,0"
    Scale="1,1,1"
   ) {
    rotationMode=BMM_Additive
    rotationSpace=BCS_BoneSpace
    alphaBoolBlend {blendOption=Linear}
    boneToModify {boneName=dyn_jacket_fr_l_MSTR}
    rotation {pitch=30}
   }
   n9 = "Transform (Modify) Bone - Bone: dyn_jacket_fr_r_MSTR"(ComponentPose=@n8, Alpha=@n7, Translation="0,0,0", Rotation="50.148,-45.826,-63.763", Scale="1,1,1") {
    rotationMode=BMM_Additive
    rotationSpace=BCS_BoneSpace
    translationMode=BMM_Additive
    alphaBoolBlend {blendOption=Linear}
    boneToModify {boneName=dyn_jacket_fr_r_MSTR}
    rotation=<rot 50.1478 -45.8265 -63.7634>
   }
   n10 = "Transform (Modify) Bone - Bone: dyn_jacket_r_MSTR"(ComponentPose=@n9, Alpha=@n7, Translation="-1.402,0.783,7.403", Rotation="19.166,-4.068,32.632", Scale="1,1,1") {
    rotationMode=BMM_Additive
    rotationSpace=BCS_BoneSpace
    translationMode=BMM_Additive
    alphaBoolBlend {blendOption=Linear}
    boneToModify {boneName=dyn_jacket_r_MSTR}
    rotation=<rot 19.166 -4.068 32.6323>
    translation=<-1.402 0.783 7.403>
   }
   n11 = "Transform (Modify) Bone - Bone: dyn_jacket_bk_r_MSTR"(ComponentPose=@n10, Alpha=@n7, Translation="-2.512,-3.528,3.739", Rotation="0,0,0", Scale="1,1,1") {
    rotationMode=BMM_Additive
    rotationSpace=BCS_BoneSpace
    translationMode=BMM_Additive
    alphaBoolBlend {blendOption=Linear}
    boneToModify {boneName=dyn_jacket_bk_r_MSTR}
    translation=<-2.512 -3.528 3.739>
   }
   n12 = "Transform (Modify) Bone - Bone: dyn_jacket_bk_MSTR"(ComponentPose=@n11, Alpha=@n7, Translation="0.000,-1.959,0.185", Rotation="0,0,0", Scale="1,1,1") {
    rotationMode=BMM_Additive
    rotationSpace=BCS_BoneSpace
    translationMode=BMM_Additive
    alphaBoolBlend {blendOption=Linear}
    boneToModify {boneName=dyn_jacket_bk_MSTR}
    translation {y=-1.959, z=0.185}
   }
   n13 = "Transform (Modify) Bone - Bone: dyn_jacket_fr_l_MSTR"(
    ComponentPose=@n12
    Alpha=Get CrouchAlpha() -> (CrouchAlpha:double)
    Translation="2.103,3.028,-4.738"
    Rotation="74.315,-70.539,-72.508"
    Scale="1,1,1"
   ) {
    rotationMode=BMM_Additive
    rotationSpace=BCS_BoneSpace
    scaleMode=BMM_Replace
    translationMode=BMM_Additive
    alphaBoolBlend {blendInTime=0.01, blendOption=Cubic, blendOutTime=0.01}
    boneToModify {boneName=dyn_jacket_fr_l_MSTR}
    rotation=<rot 74.315 -70.539 -72.508>
    translation=<2.103 3.028 -4.738>
   }
   n14 = Get Thigh R Crouch Pitch Alpha() -> (Thigh R Crouch Pitch Alpha:double)
   n15 = "Transform (Modify) Bone - Bone: dyn_jacket_r_MSTR"(ComponentPose=@n13, Alpha=@n14, Translation="-2.823,0.000,4.652", Rotation="10,0.000,0.000", Scale="1,1,1") {
    rotationMode=BMM_Additive
    rotationSpace=BCS_BoneSpace
    translationMode=BMM_Additive
    alphaBoolBlend {blendInTime=0.25, blendOption=Linear, blendOutTime=0.25}
    boneToModify {boneName=dyn_jacket_r_MSTR}
    rotation {pitch=10}
    translation {x=-2.823, z=4.652}
   }
   n16 = "Transform (Modify) Bone - Bone: dyn_jacket_fr_r_MSTR"(ComponentPose=@n15, Alpha=@n14, Translation="0,0,0", Rotation="10,0.000,0.000", Scale="1,1,1") {
    rotationMode=BMM_Additive
    rotationSpace=BCS_BoneSpace
    translationMode=BMM_Additive
    alphaBoolBlend {blendInTime=0.25, blendOption=Linear, blendOutTime=0.25}
    boneToModify {boneName=dyn_jacket_fr_r_MSTR}
    rotation {pitch=10}
   }
   n17 = "Transform (Modify) Bone - Bone: dyn_jacket_fr_r_MSTR"(
    ComponentPose=@n16
    Alpha=Get Thigh R Crouch Rot Alpha() -> (Thigh R Crouch Rot Alpha:double)
    Translation="-1.437,2.948,2.063"
    Rotation="20,0,0"
    Scale="1,1,1"
   ) {
    rotationMode=BMM_Additive
    rotationSpace=BCS_BoneSpace
    translationMode=BMM_Additive
    alphaBoolBlend {blendInTime=0.25, blendOption=Linear, blendOutTime=0.25}
    boneToModify {boneName=dyn_jacket_fr_r_MSTR}
    rotation {pitch=20}
    translation=<-1.437 2.948 2.063>
   }
   n18 = RigidBody(ComponentPose=@n17) {
    bClampLinearTranslationLimitToRefPose=true
    bOverrideWorldGravity=true
    overridePhysicsAsset=&M_MED_Critter_Frenzy_Physics
    simulationSpace=BaseBoneSpace
    alphaBoolBlend {blendOption=Linear}
    baseBoneRef {boneName=pelvis}
    componentAppliedLinearAccClamp=<2000 2000 2000>
    componentLinearAccScale=<0.3 0.3 0.3>
    componentLinearVelScale=<0.3 0.3 0.3>
    overrideWorldGravity {z=-980}
   }
   n19 = Get CrouchAlpha()
   n20 = "Transform (Modify) Bone - Bone: dyn_jacket_bk_l_MSTR"(ComponentPose=@n18, Alpha=@n19, Translation="0,0,0", Rotation="-11.394,0.000,0.000", Scale="1,1,1") {
    rotationMode=BMM_Additive
    rotationSpace=BCS_BoneSpace
    translationMode=BMM_Additive
    alphaBoolBlend {blendInTime=0.1, blendOption=Linear, blendOutTime=0.1}
    boneToModify {boneName=dyn_jacket_bk_l_MSTR}
    rotation {pitch=-11.394}
   }
   n21 = "Transform (Modify) Bone - Bone: dyn_jacket_l_01"(ComponentPose=@n20, Alpha=@n19, Translation="1.503,0.325,0.313", Rotation="-15,0.000,0.000", Scale="1,1,1") {
    rotationMode=BMM_Additive
    rotationSpace=BCS_BoneSpace
    translationMode=BMM_Additive
    alphaBoolBlend {blendInTime=0.1, blendOption=Linear, blendOutTime=0.1}
    alphaScaleBiasClamp {bInterpResult=true}
    boneToModify {boneName=dyn_jacket_l_01}
    rotation {pitch=-15}
    translation=<1.503 0.325 0.313>
   }
   n22 = "Transform (Modify) Bone - Bone: dyn_jacket_l_02"(ComponentPose=@n21, Alpha=@n19, Translation="0,0,0", Rotation="-20.000,10,-0.001", Scale="1,1,1") {
    rotationMode=BMM_Additive
    rotationSpace=BCS_BoneSpace
    translationMode=BMM_Additive
    alphaBoolBlend {blendInTime=0.1, blendOption=Linear, blendOutTime=0.1}
    alphaScaleBiasClamp {bInterpResult=true}
    boneToModify {boneName=dyn_jacket_l_02}
    rotation=<rot -20 10 -0.001>
   }
   n23 = "Transform (Modify) Bone - Bone: dyn_jacket_l_03"(ComponentPose=@n22, Alpha=@n19, Translation="0,0,0", Rotation="0,0,-13.091", Scale="1,1,1") {
    rotationMode=BMM_Additive
    rotationSpace=BCS_BoneSpace
    translationMode=BMM_Additive
    alphaBoolBlend {blendInTime=0.1, blendOption=Linear, blendOutTime=0.1}
    alphaScaleBiasClamp {bInterpResult=true}
    boneToModify {boneName=dyn_jacket_l_03}
    rotation {roll=-13.091}
   }
   n24 = "Transform (Modify) Bone - Bone: dyn_jacket_fr_r_MSTR"(ComponentPose=@n23, Alpha=@n19, Translation="0,0,0", Rotation="5,0,0", Scale="1,1,1") {
    rotationMode=BMM_Additive
    rotationSpace=BCS_BoneSpace
    translationMode=BMM_Additive
    alphaBoolBlend {blendInTime=0.1, blendOption=Linear, blendOutTime=0.1}
    boneToModify {boneName=dyn_jacket_fr_r_MSTR}
    rotation {pitch=5}
   }
   n25 = "Transform (Modify) Bone - Bone: dyn_jacket_r_MSTR"(ComponentPose=@n24, Alpha=@n19, Translation="0,0,0", Rotation="9,0,0", Scale="1,1,1") {
    rotationMode=BMM_Additive
    rotationSpace=BCS_BoneSpace
    translationMode=BMM_Additive
    alphaBoolBlend {blendInTime=0.1, blendOption=Linear, blendOutTime=0.1}
    boneToModify {boneName=dyn_jacket_r_MSTR}
    rotation {pitch=9}
   }
   n26 = "Transform (Modify) Bone - Bone: dyn_jacket_fr_r_MSTR"(ComponentPose=@n25, Alpha=1, Translation="0,0,0", Rotation="5,0,0", Scale="1,1,1") {
    rotationMode=BMM_Additive
    rotationSpace=BCS_BoneSpace
    translationMode=BMM_Additive
    alphaBoolBlend {blendInTime=0.1, blendOption=Linear, blendOutTime=0.1}
    boneToModify {boneName=dyn_jacket_fr_r_MSTR}
    rotation {pitch=5}
   }
   n27 = "Transform (Modify) Bone - Bone: dyn_jacket_fr_l_MSTR"(ComponentPose=@n26, Alpha=@n19, Translation="0.021,5.097,-0.344", Rotation="10,7.193,29.682", Scale="0.9,0.9,0.9") {
    rotationMode=BMM_Additive
    rotationSpace=BCS_BoneSpace
    scaleMode=BMM_Replace
    translationMode=BMM_Additive
    alphaBoolBlend {blendInTime=0.01, blendOption=Linear, blendOutTime=0.01}
    boneToModify {boneName=dyn_jacket_fr_l_MSTR}
    rotation=<rot 10 7.193 29.682>
    scale=<0.9 0.9 0.9>
    translation=<0.021 5.097 -0.344>
   }
   n28 = "Transform (Modify) Bone - Bone: dyn_jacket_fr_l_01"(ComponentPose=@n27, Alpha=@n19, Translation="0,0,0", Rotation="0,0,0", Scale="0.8,0.8,0.8") {
    rotationMode=BMM_Additive
    rotationSpace=BCS_BoneSpace
    scaleMode=BMM_Replace
    scaleSpace=BCS_BoneSpace
    translationMode=BMM_Additive
    alphaBoolBlend {blendInTime=0.01, blendOption=Linear, blendOutTime=0.01}
    boneToModify {boneName=dyn_jacket_fr_l_01}
    scale=<0.8 0.8 0.8>
   }
   n29 = "Transform (Modify) Bone - Bone: dyn_jacket_fr_l_02"(ComponentPose=@n28, Alpha=@n19, Translation="0.347,-0.034,-2", Rotation="0,0,0", Scale="0.9,0.9,0.9") {
    rotationMode=BMM_Additive
    rotationSpace=BCS_BoneSpace
    scaleMode=BMM_Replace
    scaleSpace=BCS_BoneSpace
    translationMode=BMM_Additive
    alphaBoolBlend {blendInTime=0.01, blendOption=Linear, blendOutTime=0.01}
    boneToModify {boneName=dyn_jacket_fr_l_02}
    scale=<0.9 0.9 0.9>
    translation=<0.347 -0.034 -2>
   }
   n30 = "Transform (Modify) Bone - Bone: dyn_jacket_fr_l_03"(ComponentPose=@n29, Alpha=@n19, Translation="0,0,0", Rotation="-45,0,-40", Scale="0.9,0.900,0.900") {
    rotationMode=BMM_Additive
    rotationSpace=BCS_BoneSpace
    scaleMode=BMM_Replace
    scaleSpace=BCS_BoneSpace
    translationMode=BMM_Additive
    alphaBoolBlend {blendInTime=0.01, blendOption=Linear, blendOutTime=0.01}
    alphaScaleBiasClamp {bInterpResult=true, interpSpeedIncreasing=4}
    boneToModify {boneName=dyn_jacket_fr_l_03}
    rotation {pitch=-45, roll=-40}
    scale=<0.9 0.9 0.9>
   }
   n31 = Get Thigh R Crouch Pitch Alpha()
   n32 = "Transform (Modify) Bone - Bone: dyn_jacket_fr_r_01"(ComponentPose=@n30, Alpha=@n31, Translation="0,0,0", Rotation="20,-19.280,-12.254", Scale="1,1,1") {
    rotationMode=BMM_Additive
    rotationSpace=BCS_BoneSpace
    translationMode=BMM_Additive
    alphaBoolBlend {blendInTime=0.25, blendOption=Linear, blendOutTime=0.25}
    boneToModify {boneName=dyn_jacket_fr_r_01}
    rotation=<rot 20 -19.28 -12.254>
   }
   n33 = "Transform (Modify) Bone - Bone: dyn_jacket_fr_r_02"(ComponentPose=@n32, Alpha=@n31, Translation="0,0,0", Rotation="-20,0,0", Scale="1,1,1") {
    rotationMode=BMM_Additive
    rotationSpace=BCS_BoneSpace
    translationMode=BMM_Additive
    alphaBoolBlend {blendInTime=0.25, blendOption=Linear, blendOutTime=0.25}
    boneToModify {boneName=dyn_jacket_fr_r_02}
    rotation {pitch=-20}
   }
   n34 = "Transform (Modify) Bone - Bone: dyn_jacket_fr_r_02"(ComponentPose=@n33, Alpha=@n31, Translation="0,0,0", Rotation="0,30,0.000", Scale="1,1,1") {
    rotationMode=BMM_Additive
    rotationSpace=BCS_BoneSpace
    translationMode=BMM_Additive
    alphaBoolBlend {blendInTime=0.25, blendOption=Linear, blendOutTime=0.25}
    boneToModify {boneName=dyn_jacket_fr_r_02}
    rotation {yaw=30}
   }
   n35 = "Transform (Modify) Bone - Bone: dyn_jacket_fr_r_03"(ComponentPose=@n34, Alpha=@n31, Translation="0,0,0", Rotation="-10,0,0.000", Scale="1,1,1") {
    rotationMode=BMM_Additive
    rotationSpace=BCS_BoneSpace
    translationMode=BMM_Additive
    alphaBoolBlend {blendInTime=0.25, blendOption=Linear, blendOutTime=0.25}
    boneToModify {boneName=dyn_jacket_fr_r_03}
    rotation {pitch=-10}
   }
   n36 = "Transform (Modify) Bone - Bone: dyn_jacket_fr_l_01"(ComponentPose=@n35, Alpha=@n6, Translation="0,0,0", Rotation="15,0,0", Scale="1,1,1") {
    rotationMode=BMM_Additive
    rotationSpace=BCS_BoneSpace
    translationMode=BMM_Additive
    alphaBoolBlend {blendInTime=0.25, blendOption=Linear, blendOutTime=0.25}
    alphaScaleBiasClamp {bInterpResult=true, interpSpeedIncreasing=6}
    boneToModify {boneName=dyn_jacket_fr_l_01}
    rotation {pitch=15}
   }
   n37 = "Transform (Modify) Bone - Bone: dyn_jacket_l_MSTR"(ComponentPose=@n36, Alpha=@n6, Translation="0,0,0", Rotation="40,0,0", Scale="1,1,1") {
    rotationMode=BMM_Additive
    rotationSpace=BCS_BoneSpace
    translationMode=BMM_Additive
    alphaBoolBlend {blendInTime=0.25, blendOption=Linear, blendOutTime=0.25}
    alphaScaleBiasClamp {bInterpResult=true, interpSpeedIncreasing=20}
    boneToModify {boneName=dyn_jacket_l_MSTR}
    rotation {pitch=40}
   }
   n38 = "Transform (Modify) Bone - Bone: dyn_jacket_fr_r_02"(ComponentPose=@n37, Alpha=@n6, Translation="-1.615,-1.924,0.313", Rotation="9.770,-9.999,0.000", Scale="1,1,1") {
    rotationMode=BMM_Additive
    rotationSpace=BCS_BoneSpace
    translationMode=BMM_Additive
    alphaBoolBlend {blendInTime=0.25, blendOption=Linear, blendOutTime=0.25}
    boneToModify {boneName=dyn_jacket_fr_r_02}
    rotation {pitch=9.77, yaw=-9.999}
    translation=<-1.615 -1.924 0.313>
   }
   n39 = "Transform (Modify) Bone - Bone: dyn_jacket_fr_l_MSTR"(
    ComponentPose=@n38
    Alpha=Get Thigh L Crouch Rot Alpha() -> (Thigh L Crouch Rot Alpha:double)
    Translation="0,0,0"
    Rotation="-30,0,-30"
    Scale="1,1,1"
   ) {
    rotationMode=BMM_Additive
    rotationSpace=BCS_BoneSpace
    translationMode=BMM_Additive
    alphaBoolBlend {blendInTime=0.25, blendOption=Linear, blendOutTime=0.25}
    alphaScaleBiasClamp {bInterpResult=true, interpSpeedIncreasing=6}
    boneToModify {boneName=dyn_jacket_fr_l_MSTR}
    rotation {pitch=-30, roll=-30}
   }
   n40 = "Transform (Modify) Bone - Bone: dyn_jacket_l_MSTR"(
    ComponentPose=@n39
    Alpha=Get Thigh L Crouch Pitch In Alpha() -> (Thigh L Crouch Pitch In Alpha:double)
    Translation="0,0,0"
    Rotation="-10,0,0"
    Scale="1,1,1"
   ) {
    rotationMode=BMM_Additive
    rotationSpace=BCS_BoneSpace
    translationMode=BMM_Additive
    alphaBoolBlend {blendInTime=0.25, blendOption=Linear, blendOutTime=0.25}
    alphaScaleBiasClamp {bInterpResult=true, interpSpeedIncreasing=6}
    boneToModify {boneName=dyn_jacket_l_MSTR}
    rotation {pitch=-10}
   }
   n41 = Get Thigh R Crouch Rot Alpha()
   n42 = "Transform (Modify) Bone - Bone: dyn_jacket_fr_r_02"(ComponentPose=@n40, Alpha=@n41, Translation="0,0,0", Rotation="-13.164,10.304,-7.207", Scale="1,1,1") {
    rotationMode=BMM_Additive
    rotationSpace=BCS_BoneSpace
    translationMode=BMM_Additive
    alphaBoolBlend {blendInTime=0.25, blendOption=Linear, blendOutTime=0.25}
    boneToModify {boneName=dyn_jacket_fr_r_02}
    rotation=<rot -13.164 10.3041 -7.207>
   }
   n43 = "Transform (Modify) Bone - Bone: dyn_jacket_fr_r_03"(ComponentPose=@n42, Alpha=@n41, Translation="-1.208,2.477,0.572", Rotation="-20.740,-29.203,-0.016", Scale="1,1,1") {
    rotationMode=BMM_Additive
    rotationSpace=BCS_BoneSpace
    translationMode=BMM_Additive
    alphaBoolBlend {blendInTime=0.25, blendOption=Linear, blendOutTime=0.25}
    boneToModify {boneName=dyn_jacket_fr_r_03}
    rotation=<rot -20.74 -29.203 -0.016>
    translation=<-1.208 2.477 0.572>
   }
   n44 = "Transform (Modify) Bone - Bone: dyn_jacket_fr_r_03"(ComponentPose=@n43, Alpha=Get Thigh L Crouch Pitch Alpha(), Translation="0,0,0", Rotation="0,0,0", Scale="1,1,1") {
    rotationMode=BMM_Additive
    rotationSpace=BCS_BoneSpace
    translationMode=BMM_Additive
    alphaBoolBlend {blendInTime=0.25, blendOption=Linear, blendOutTime=0.25}
    boneToModify {boneName=dyn_jacket_fr_r_03}
   }
   n45 = "Transform (Modify) Bone - Bone: dyn_jacket_fr_r_03"(ComponentPose=@n44, Alpha=Get Jump Alpha(), Translation="0,0,0", Rotation="-15,0,0", Scale="1,1,1") {
    rotationMode=BMM_Additive
    rotationSpace=BCS_BoneSpace
    translationMode=BMM_Additive
    alphaBoolBlend {blendInTime=0.25, blendOption=Linear, blendOutTime=0.25}
    boneToModify {boneName=dyn_jacket_fr_r_03}
    rotation {pitch=-15}
   }
   n46 = "Transform (Modify) Bone - Bone: dyn_jacket_r_02"(ComponentPose=@n45, Alpha=1, Translation="0,0,0", Rotation="-10,0,0", Scale="1,1,1") {
    rotationMode=BMM_Additive
    rotationSpace=BCS_BoneSpace
    translationMode=BMM_Additive
    alphaBoolBlend {blendInTime=0.25, blendOption=Linear, blendOutTime=0.25}
    boneToModify {boneName=dyn_jacket_r_02}
    rotation {pitch=-10}
   }
   Output Pose(Result=Component To Local(ComponentPose=@n46)) {layerGroup=None}
  ]}
 EventGraph {nodes [
   Event BlueprintUpdateAnimation()
   > n1 = Sequence!(execute) -> (then_0, then_1, then_2:Exec, then_3:Exec)
   > Set Jump Falling(Jump Falling:bool=OR Boolean(A=Get bIsJumping(), B=Get bIsFalling())) -> (Output_Get:bool)
   n2 = Branch(execute=@n1.then_1, Condition=Get Jump Falling() -> (Jump Falling:bool))
   n3 = BreakTransform(InTransform=GetDeltaTransformFromRefPose(self=GetOwnerSkeletalMeshComponent(), BoneName=thigh_r, BaseName=pelvis))
   > Set Jump Alpha(Jump Alpha:double=MapRangeClamped(Value=BreakRotator(InRot=@n3.Rotation).Yaw, InRangeA=10, InRangeB=-100)) -> (Output_Get:double)
   Set Jump Alpha(execute=@n2.else, Jump Alpha=0)
   n4 = Branch(execute=@n1.then_2, Condition=Get bIsCrouching())
   > Set CrouchAlpha(CrouchAlpha:double=1) -> (Output_Get:double)
   n5 = GetOwnerSkeletalMeshComponent()
   n6 = BreakRotator(InRot=BreakTransform(InTransform=GetDeltaTransformFromRefPose(self=@n5, BoneName=thigh_r, BaseName=pelvis)).Rotation)
   > Set Thigh R Crouch Rot Roll Alpha(Thigh R Crouch Rot Roll Alpha:double=MapRangeClamped(Value=@n6, InRangeA=0, InRangeB=45)) -> (Output_Get:double)
   > Set Thigh R Crouch Pitch Alpha(Thigh R Crouch Pitch Alpha:double=MapRangeClamped(Value=@n6.Pitch, InRangeA=15, InRangeB=40)) -> (Output_Get:double)
   > Set Thigh R Crouch Rot Alpha(Thigh R Crouch Rot Alpha:double=MapRangeClamped(Value=@n6.Yaw, InRangeA=-25, InRangeB=-105)) -> (Output_Get:double)
   n7 = GetOwnerSkeletalMeshComponent()
   n8 = BreakRotator(InRot=BreakTransform(InTransform=GetDeltaTransformFromRefPose(self=@n7, BoneName=thigh_l, BaseName=pelvis)).Rotation)
   > Set Thigh L Crouch Pitch Alpha(Thigh L Crouch Pitch Alpha:double=MapRangeClamped(Value=@n8.Pitch, InRangeA=25, InRangeB=45)) -> (Output_Get:double)
   > Set Thigh L Crouch Rot Alpha(Thigh L Crouch Rot Alpha:double=MapRangeClamped(Value=@n8.Yaw, InRangeA=-100, InRangeB=-40)) -> (Output_Get:double)
   > Set Thigh L Crouch Pitch In Alpha(Thigh L Crouch Pitch In Alpha:double=MapRangeClamped(Value=@n8, InRangeA=0, InRangeB=60)) -> (Output_Get:double)
   Set Thigh R Pitch Alpha(execute=@n4.else, Thigh R Pitch Alpha:double=0) -> (Output_Get:double)
   > Set Thigh R Rot Alpha(Thigh R Rot Alpha:double=0) -> (Output_Get:double)
   > Set Thigh L Crouch Pitch Alpha(Thigh L Crouch Pitch Alpha=0)
   > Set Thigh L Crouch Rot Alpha(Thigh L Crouch Rot Alpha=0)
   > Set Thigh R Crouch Pitch Alpha(Thigh R Crouch Pitch Alpha=0)
   > Set Thigh R Crouch Rot Alpha(Thigh R Crouch Rot Alpha=0)
   > Set Thigh L Crouch Pitch In Alpha(Thigh L Crouch Pitch In Alpha=0)
   > Set CrouchAlpha(CrouchAlpha=MapRangeClamped(Value=@n8.Yaw, InRangeA=-60, InRangeB=-100))
  ]}
]
variables [
 CrouchAlpha {type=number}
 Jump Alpha {type=number}
 Jump Falling {type=boolean}
 Thigh L Crouch Pitch Alpha {type=number}
 Thigh L Crouch Pitch In Alpha {type=number}
 Thigh L Crouch Rot Alpha {type=number}
 Thigh R Crouch Pitch Alpha {type=number}
 Thigh R Crouch Rot Alpha {type=number}
 Thigh R Crouch Rot Roll Alpha {type=number}
 Thigh R Pitch Alpha {type=number}
 Thigh R Rot Alpha {type=number}
]
